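; Gameplay ability for the RocketDrill item that blocks the avatar player's movement input, by its name after a hard impact.
; On activation it caches the valid Fort player pawn and sets ignore-move-input on that pawn's controller.
; A one-second delay acts as a failsafe in case the granting effect is not removed, then removes that effect and ends the ability.
; OnEndAbility releases the input block, whether cancelled or not.
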(event Ability|EventActivateAbility
  (|CallParentFunction)
  (bind _outfortplayerpawn (CallFunction|GetAvatarasValidFortPlayerPawn)
    (:ValidFortPlayerPawn
      (Variables|Default|SetPlayerPawn _outfortplayerpawn)
      (Input|SetIgnoreMoveInput (Fortnite|GetFortPlayerControllerfromActor _outfortplayerpawn) true)
      ; failsafe in case of no GE removal
      (Utilities|FlowControl|Delay 1.0)
      (Ability|RemoveGrantedbyEffect)
      (Ability|EndAbility))))

(event Ability|EventOnEndAbility
  (params (bWasCancelled :type Boolean))
  (|CallParentFunction bWasCancelled)
  (Input|SetIgnoreMoveInput (Fortnite|GetFortPlayerControllerfromActor (Variables|Default|GetPlayerPawn)) false))
