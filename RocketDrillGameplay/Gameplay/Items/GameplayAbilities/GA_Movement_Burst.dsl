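; Rocket Drill burst movement ability that launches the player along their aim, smashing through pawns, buildings and vehicles.
; An enum state machine runs startup, slide root-motion boost, falling, then skid stop or landing impact based on fall angle.
; A timer samples velocity and sphere-traces ahead while boosting to knock back and damage targets, stopping after a structure limit.
; Gameplay effects gate actions; camera changes and a temporary net-correction tolerance support the move.
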(event Ability|EventActivateAbilityFromEvent
  (params (EventData :type Gameplay%20Event%20Data%20Structure%20%28by%20ref%29))
  (|CallParentFunction EventData)
  (bind _returnvalue (Ability|CommitAbilityCost false))
  (if _returnvalue
    (bind _outfortplayerpawn (CallFunction|GetAvatarasValidFortPlayerPawn)
      (:ValidFortPlayerPawn
        (Variables|Default|SetPlayerPawn _outfortplayerpawn)
        (bind _asb_weapon_rocket_drill (Utilities|Casting|CastToB_Weapon_RocketDrill (Ability|GetCurrentSourceWeapon)))
        (Variables|Default|SetWeapon _asb_weapon_rocket_drill)
        (Variables|Default|SetStructuresDestroyed 0)
        (Variables|Default|SetbWillImpact false)
        (Variables|Default|SetbPostLaunch false)
        (CallFunction|SetState "NewEnumerator0")))))

(event Custom|OnChangeState
  (params (State :type ERocketDrillState%20Enum :default "NewEnumerator0"))
  (bind (_returnvalue_location _returnvalue_rotation) (Math|Transform|BreakTransform (Targeting|GetTargetingTransform (Variables|Default|GetPlayerPawn) "PawnTowardsFocus" 0)))
  (bind _playerpawn (Variables|Default|GetPlayerPawn))
  (switch Utilities|FlowControl|Switch|SwitchonERocketDrillState State
    (:NewEnumerator0
      (bind _returnvalue (Ability|ApplyGameplayEffectToOwner "/RocketDrillGameplay/Gameplay/Items/GameplayEffects/GE_RocketDrill_Friction_TriBurst.GE_RocketDrill_Friction_TriBurst_C"))
      (Variables|Default|SetGE_Friction_Handle _returnvalue)
      (bind _returnvalue_2 (Ability|ApplyGameplayEffectToOwner "/RocketDrillGameplay/Gameplay/Items/GameplayEffects/GE_InStartupState.GE_InStartupState_C"))
      (Variables|Default|SetGE_StartupState_Handle _returnvalue_2)
      (bind _returnvalue_3 (Ability|ApplyGameplayEffectToOwner "/RocketDrillGameplay/Gameplay/Items/GameplayEffects/GE_BlockActionsWhileTraversing.GE_BlockActionsWhileTraversing_C"))
      (Variables|Default|SetGE_BlockActionsTraversing_Handle _returnvalue_3)
      (bind _returnvalue_4 (Ability|ApplyGameplayEffectToOwner "/RocketDrillGameplay/Gameplay/Items/GameplayEffects/GE_BlockActionsWhileRecovering.GE_BlockActionsWhileRecovering_C"))
      (Variables|Default|SetGE_BlockActionsRecovery_Handle _returnvalue_4)
      (Input|SetIgnoreMoveInput (Fortnite|GetFortPlayerControllerfromActor _playerpawn) true)
      (bind _asfort_movement_comp_character (Utilities|Casting|CastToFortMovementComp_Character (Variables|/Script/Engine.Character|GetCharacterMovement _playerpawn)))
      (NetTolerance|SetUseHighAccelerationTolerance _asfort_movement_comp_character true "RocketDrillMovement") ; TODO this is to temporarily hide net corrections, will need to be removed eventually
      (bind _returnvalue_5 (Game|Player|AddNewCameraModifier (Variables|/Script/Engine.PlayerController|GetPlayerCameraManager (Fortnite|GetFortPlayerControllerfromActor (Variables|Default|GetPlayerPawn))) "/RocketDrillGameplay/Gameplay/Items/CameraModifier_RocketDrill_TraversalVersion.CameraModifier_RocketDrill_TraversalVersion_C"))
      (bind _output_get (Variables|Default|SetCameraMod _returnvalue_5))
      (CameraModifier|EnableModifier _output_get)
      (Camera|SetBase3PCamera (Ability|GetCurrentSourceWeapon) (Variables|Default|GetCameraMode_NoCollisionCheck))
      (Utilities|FlowControl|Delay (Variables|Default|GetStartupTime))
      (CallFunction|SetState "NewEnumerator1"))
    (:NewEnumerator1
      (Ability|ApplyGameplayEffectToOwner "/Game/Athena/Items/Parents/FallImmunity/GE_Athena_TillLandFallDamageImmunity_NoFX_Grant_Parent.GE_Athena_TillLandFallDamageImmunity_NoFX_Grant_Parent_C")
      (Ability|RemoveGameplayEffectFromOwnerWithHandle (Variables|Default|GetGE_StartupState_Handle))
      (Ability|AddGameplayCueToOwner :GameplayCueTag (Variables|Default|GetBoostingCue) :bRemoveOnAbilityEnd false)
      (bind _returnvalue_6 (Utilities|Time|SetTimerbyEvent (EventDispatchers|CreateEvent :!event Custom|Update) 0.05 true true))
      (Variables|Default|SetUpdateHandle _returnvalue_6)
      (Ability|Tasks|WaitNetSync "BothWait")
      (bind _asynctaskproxy (Ability|Tasks|ApplyRootMotion_Slide "None" 0.75 (Math|Vector|GetForwardVector _returnvalue_rotation) (/ (Math|Float|MapRangeClamped (Math|Vector|DotProduct (Math|Vector|GetForwardVector _returnvalue_rotation) "0.000000,0.000000,1.000000") 0.0 1.0 3e+03 2e+03) 1.2) (Math|Float|MapRangeClamped (Math|Vector|DotProduct (Math|Vector|GetForwardVector _returnvalue_rotation) "0.000000,0.000000,1.000000") 0.0 1.0 3e+03 2e+03) "/RocketDrillGameplay/Gameplay/Items/Curve_TriBurstVelocity.Curve_TriBurstVelocity" 4e+01 0 "MaintainLastRootMotionVelocity" "0, 0, 0" 0.0)
        (:then
          (Variables|Default|SetRootMotionTask _asynctaskproxy))
        (:OnFinish
          (Variables|Default|SetbPostLaunch true)
          (if (GameplayTags|HasMatchingGameplayTag (Variables|Default|GetPlayerPawn) "(TagName=\"MovementMode.Falling\")")
            (CallFunction|SetState "NewEnumerator2")
            ; allow reactivation, but NOT any other actions
            (if (ScalableFloat|GetValueasBool (Variables|Default|GetSF_AllowAirActivation))
              (Ability|RemoveGameplayEffectFromOwnerWithHandle (Variables|Default|GetGE_BlockActionsTraversing_Handle)))
            (else
              (CallFunction|SetState "NewEnumerator3")
              (Ability|RemoveGameplayEffectFromOwnerWithHandle (Variables|Default|GetGE_BlockActionsTraversing_Handle)))))))
    (:NewEnumerator2
      (Ability|Tasks|WaitGameplayTagQuery "(TokenStreamVersion=0,TagDictionary=((TagName=\"MovementMode.Falling\")),QueryTokenStream=(0,1,2,1,0),UserDescription=\"\",AutoDescription=\" ALL( MovementMode.Falling )\")" 0 "WhenFalse" true
        (:Triggered
          (if (Variables|Default|GetbWillImpact)
            (CallFunction|SetState "NewEnumerator4")
            (else
              (CallFunction|SetState "NewEnumerator3"))))))
    (:NewEnumerator3
      (bind _returnvalue_7 (Ability|ApplyGameplayEffectToOwner "/RocketDrillGameplay/Gameplay/Items/GameplayEffects/GE_SkidToStop.GE_SkidToStop_C"))
      (Variables|Default|SetGE_Skidding_Handle _returnvalue_7)
      (Utilities|FlowControl|Delay (ScalableFloat|GetValueatLevel (Variables|Default|GetSF_SlideTime)))
      (Ability|RemoveGameplayEffectFromOwnerWithHandle (Variables|Default|GetGE_Friction_Handle))
      (Ability|RemoveGameplayEffectFromOwnerWithHandle (Variables|Default|GetGE_Skidding_Handle))
      ; allow any actions (weapon swap, etc)
      (Ability|RemoveGameplayEffectFromOwnerWithHandle (Variables|Default|GetGE_BlockActionsRecovery_Handle))
      (Ability|EndAbility))
    (:NewEnumerator4
      (Ability|RemoveGameplayEffectFromOwnerWithHandle (Variables|Default|GetGE_Friction_Handle))
      (Ability|ApplyGameplayEffectToOwner "/RocketDrillGameplay/Gameplay/Items/GameplayEffects/GE_LandingImpact.GE_LandingImpact_C")
      (Ability|EndAbility))))

(event Custom|Update
  (Variables|Default|SetActualVelocity (/ (- (Transformation|GetActorLocation (Variables|Default|GetPlayerPawn)) (Variables|Default|GetLastPosition)) (- (Utilities|Time|GetGameTimeinSeconds) (Variables|Default|GetLastTime))))
  (Utilities|FlowControl|Sequence
    (:then_0
      (if (or (Utilities|Enum|Equal(Enum) (Variables|Default|GeteDrillState) "NewEnumerator1") (Variables|/RocketDrillGameplay/Gameplay/Items/B_Weapon_RocketDrill.B_Weapon_RocketDrill_C|GetCheatAlwaysDestroy (Variables|Default|GetWeapon)))
        (bind _outhits (Collision|MultiSphereTraceByProfile (Transformation|GetActorLocation (Variables|Default|GetPlayerPawn)) (+ (Transformation|GetActorLocation (Variables|Default|GetPlayerPawn)) (* (Math|Vector|Normalize (Variables|Default|GetActualVelocity)) (ScalableFloat|GetValueatLevel (Variables|Default|GetSF_EnvironmentDamageTrace_Distance)))) (ScalableFloat|GetValueatLevel (Variables|Default|GetSF_EnvironmentDamageTrace_Radius)) "OverlapAll" false (Utilities|Array|MakeArray (Variables|Default|GetPlayerPawn)) "None" true "(R=1.000000,G=0.000000,B=0.000000,A=1.000000)" "(R=0.000000,G=1.000000,B=0.000000,A=1.000000)" 0.2))
        (for _array_element _outhits
          (bind _asfort_pawn (Utilities|Casting|CastToFortPawn (.HitActor (Collision|BreakHitResult _array_element)))
            (:then
              (if (not (GameplayTags|HasMatchingGameplayTag _asfort_pawn "(TagName=\"Granted.Athena.RocketDrill.HitByDrill\")"))
                (bind (_returnvalue_x _returnvalue_y) (Math|Vector|BreakVector (* (Math|Vector|Normalize (- (Transformation|GetActorLocation _asfort_pawn) (Transformation|GetActorLocation (Variables|Default|GetPlayerPawn)))) (Math|Float|SelectFloat 3e+03 2e+03 (GameplayTags|HasMatchingGameplayTag (Variables|Default|GetPlayerPawn) "(TagName=\"Granted.Athena.RocketDrill.HasWeaponVersion\")")))))
                (Character|LaunchCharacter _asfort_pawn (Math|Vector|MakeVector _returnvalue_x _returnvalue_y 1.2e+03) false false)
                (Ability|ApplyGameplayEffectSpecToTarget (Ability|MakeOutgoingGameplayEffectSpec "/RocketDrillGameplay/Gameplay/Items/GameplayEffects/GE_RocketDrill_DamageOverTime.GE_RocketDrill_DamageOverTime_C") (Ability|TargetData|AbilityTargetDatafromActor _asfort_pawn))
                (Ability|ApplyGameplayEffectToTarget (Ability|TargetData|AbilityTargetDatafromActor _asfort_pawn) "/RocketDrillGameplay/Gameplay/Items/GameplayEffects/GE_RocketDrill_WasHit.GE_RocketDrill_WasHit_C")))
            (:CastFailed
              (bind _asbuilding_actor (Utilities|Casting|CastToBuildingActor (.HitActor (Collision|BreakHitResult _array_element)))
                (:then
                  (if (Variables|/Script/Engine.Actor|GetbCanBeDamaged _asbuilding_actor)
                    (if (>= (Variables|Default|GetStructuresDestroyed) (ScalableFloat|GetValueasInteger (Variables|Default|GetSF_MaxStructuresDestroyed)))
                      ;[0.5 0.5 0.5 0.5] Need to add additional checks here for prop size. There don't seem to be tags for that...
                      (if (not (.bSuccess (Utilities|Casting|CastToBuildingProp _asbuilding_actor)))
                        (GameplayTasks|EndTask (Variables|Default|GetRootMotionTask))
                        (CallFunction|SetState "NewEnumerator3"))
                      (else
                        (GameplayEffects|ApplyGameplayEffectToTarget :self (Ability|GetFortAbilitySystemComponent (Variables|Default|GetPlayerPawn)) :GameplayEffectClass "/RocketDrillGameplay/Gameplay/Items/GameplayEffects/GE_RocketDrill_Damage_DirectHit_AlwaysDestroy.GE_RocketDrill_Damage_DirectHit_AlwaysDestroy_C" :Target (Ability|GetFortAbilitySystemComponent _asbuilding_actor) :Level 0.0)
                        ;[0.5 0.5 0.5 0.5] Need to add additional checks here for prop size. There don't seem to be tags for that...
                        (if (not (.bSuccess (Utilities|Casting|CastToBuildingProp _asbuilding_actor)))
                          (Math|Integer|IncrementInt (Variables|Default|GetStructuresDestroyed)))))))
                (:CastFailed
                  (Utilities|Casting|CastToFortVehicleInterface (.HitActor (Collision|BreakHitResult _array_element)))
                  (GameplayEffects|ApplyGameplayEffectToTarget :self (Ability|GetFortAbilitySystemComponent (Variables|Default|GetPlayerPawn)) :GameplayEffectClass "/RocketDrillGameplay/Gameplay/Items/GameplayEffects/GE_RocketDrill_VehicleDamage.GE_RocketDrill_VehicleDamage_C" :Target (Ability|GetFortAbilitySystemComponent (.HitActor (Collision|BreakHitResult _array_element))) :Level 0.0)
                  (Audio|BroadcastAbilityImpactatLocation 0 (.HitActor (Collision|BreakHitResult _array_element)) (* (CallFunction|GetKnockbackDirection 0.7 0.5) 1e+03) 7.8e+02) ; This is the secret node that launches vehicles. Note it needs the 780 radius to work.
                )))))))
    (:then_1
      (if (not (Variables|Default|GetbWillImpact))
        (Variables|Default|SetbWillImpact (>= (Math|Vector|DotProduct (Math|Vector|Normalize (Variables|Default|GetActualVelocity)) "0.000000,0.000000,-1.000000") (Math|Trig|Sin(Degrees) (Variables|Default|GetAngleOfFallToTriggerImpact))))))
    (:then_2
      (Variables|Default|SetLastPosition (Transformation|GetActorLocation (Variables|Default|GetPlayerPawn)))
      (Variables|Default|SetLastTime (Utilities|Time|GetGameTimeinSeconds)))))

(event Ability|EventOnEndAbility
  (params (bWasCancelled :type Boolean))
  (bind _rootmotiontask (Variables|Default|GetRootMotionTask))
  (Ability|RemoveGameplayEffectFromOwnerWithHandle (Variables|Default|GetGE_StartupState_Handle))
  (Ability|RemoveGameplayEffectFromOwnerWithHandle (Variables|Default|GetGE_BlockActionsTraversing_Handle))
  (Ability|RemoveGameplayEffectFromOwnerWithHandle (Variables|Default|GetGE_BlockActionsRecovery_Handle))
  (Ability|RemoveGameplayEffectFromOwnerWithHandle (Variables|Default|GetGE_Friction_Handle))
  (Ability|RemoveGameplayEffectFromOwnerWithHandle (Variables|Default|GetGE_Skidding_Handle))
  (Input|SetIgnoreMoveInput (Fortnite|GetFortPlayerControllerfromActor (Variables|Default|GetPlayerPawn)) false)
  (Camera|SetBase3PCamera (Ability|GetCurrentSourceWeapon) (Variables|Default|GetBaseCameraMode))
  (Class|CameraModifierRocketDrillTraversalVersion|WindDownandRemove (Variables|Default|GetCameraMod))
  (Utilities|Time|ClearandInvalidateTimerbyHandle (Variables|Default|GetUpdateHandle))
  (Utilities|IsValid _rootmotiontask
    (:Is%20Valid
      (GameplayTasks|EndTask _rootmotiontask))))

(fn GetKnockbackDirection
  (params (UpAmount :type Float%20%28double-precision%29) (RightAmount :type Float%20%28single-precision%29))
  (return (Utilities|Operators|Add (Math|Vector|Normalize2D(Vector) (Math|Vector|GetForwardVector (Pawn|GetBaseAimRotation (Variables|Default|GetPlayerPawn)))) (Math|Vector|MakeVector 0.0 0.0 UpAmount) (* (Math|Vector|CrossProduct (Math|Vector|Normalize2D(Vector) (Math|Vector|GetForwardVector (Pawn|GetBaseAimRotation (Variables|Default|GetPlayerPawn)))) "0.000000,0.000000,-1.000000") RightAmount))))

(fn SetState
  (params (inState :type ERocketDrillState%20Enum :default "NewEnumerator0"))
  (bind _output_get (Variables|Default|SeteDrillState inState))
  (CallFunction|OnChangeState _output_get))
