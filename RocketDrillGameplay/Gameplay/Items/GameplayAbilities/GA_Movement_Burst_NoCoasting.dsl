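; Rocket Drill burst movement ability: after a startup hold that slows the player and ignores move input, it boosts along the targeting direction.
; Boost uses a curve-driven root motion slide, slower when aimed upward, then applies skid-to-stop and ends.
; A repeating timer sphere-traces along measured velocity to launch and damage pawns, destroy structures up to a cap and damage vehicles.
; Vehicle knockback relies on BroadcastAbilityImpactatLocation; high net acceleration tolerance temporarily masks corrections.
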(event Custom|OnChangeState
  (params (State :type ERocketDrillState%20Enum :default "NewEnumerator0"))
  (bind (_returnvalue_location _returnvalue_rotation) (Math|Transform|BreakTransform (Targeting|GetTargetingTransform (Variables|Default|GetPlayerPawn) "PawnTowardsFocus" 0)))
  (bind _playerpawn (Variables|Default|GetPlayerPawn))
  (bind _charactermovement (Variables|/Script/Engine.Character|GetCharacterMovement _playerpawn))
  (bind (_velocity_x _velocity_y _velocity_z) (Math|Vector|BreakVector (Variables|/Script/Engine.MovementComponent|GetVelocity_X _charactermovement)))
  (bind _playerpawn_2 (Variables|Default|GetPlayerPawn))
  (switch Utilities|FlowControl|Switch|SwitchonERocketDrillState State
    (:NewEnumerator0
      (Variables|/Script/Engine.CharacterMovementComponent|SetGravityScale 0.0 _charactermovement)
      (Variables|/Script/Engine.MovementComponent|SetVelocity_X (Math|Vector|MakeVector (/ _velocity_x 3.0) (/ _velocity_y 3.0) (Math|Float|Clamp(Float) _velocity_z 0.0 _velocity_z)) _charactermovement)
      (bind _returnvalue (Ability|ApplyGameplayEffectToOwner "/RocketDrillGameplay/Gameplay/Items/GameplayEffects/GE_RocketDrill_Friction_TriBurst.GE_RocketDrill_Friction_TriBurst_C"))
      (Variables|Default|SetGE_Friction_Handle _returnvalue)
      (bind _returnvalue_2 (Ability|ApplyGameplayEffectToOwner "/RocketDrillGameplay/Gameplay/Items/GameplayEffects/GE_InStartupState.GE_InStartupState_C"))
      (Variables|Default|SetGE_StartupState_Handle _returnvalue_2)
      (bind _returnvalue_3 (Ability|ApplyGameplayEffectToOwner "/RocketDrillGameplay/Gameplay/Items/GameplayEffects/GE_BlockActionsWhileTraversing.GE_BlockActionsWhileTraversing_C"))
      (Variables|Default|SetGE_BlockActionsTraversing_Handle _returnvalue_3)
      (bind _returnvalue_4 (Ability|ApplyGameplayEffectToOwner "/RocketDrillGameplay/Gameplay/Items/GameplayEffects/GE_BlockActionsWhileRecovering.GE_BlockActionsWhileRecovering_C"))
      (Variables|Default|SetGE_BlockActionsRecovery_Handle _returnvalue_4)
      (Input|SetIgnoreMoveInput (Fortnite|GetFortPlayerControllerfromActor _playerpawn_2) true)
      (bind _asfort_movement_comp_character (Utilities|Casting|CastToFortMovementComp_Character (Variables|/Script/Engine.Character|GetCharacterMovement _playerpawn_2)))
      (NetTolerance|SetUseHighAccelerationTolerance _asfort_movement_comp_character true "RocketDrillMovement") ; TODO this is to temporarily hide net corrections, will need to be removed eventually
      (Ability|ExecuteGameplayCueOnOwner (Variables|GameplayCues|GetStartupTag))
      (Camera|SetBase3PCamera (Ability|GetCurrentSourceWeapon) (Variables|Default|GetCameraMode_NoCollisionCheck))
      (Utilities|FlowControl|Delay (Variables|Default|GetStartupTime))
      (CallFunction|SetState "NewEnumerator1"))
    (:NewEnumerator1
      (Ability|ApplyGameplayEffectToOwner "/Game/Athena/Items/Parents/FallImmunity/GE_Athena_TillLandFallDamageImmunity_NoFX_Grant_Parent.GE_Athena_TillLandFallDamageImmunity_NoFX_Grant_Parent_C")
      (Ability|RemoveGameplayEffectFromOwnerWithHandle (Variables|Default|GetGE_StartupState_Handle))
      (Ability|AddGameplayCueToOwner :GameplayCueTag (Variables|GameplayCues|GetBoostingTag) :bRemoveOnAbilityEnd false)
      (bind _returnvalue_5 (Utilities|Time|SetTimerbyEvent (EventDispatchers|CreateEvent :!event Custom|Update) 0.05 true true))
      (Variables|Default|SetUpdateHandle _returnvalue_5)
      (Ability|Tasks|WaitNetSync "BothWait")
      (bind _asynctaskproxy (Ability|Tasks|ApplyRootMotion_Slide "None" 0.75 (Math|Vector|GetForwardVector _returnvalue_rotation) (Math|Float|MapRangeClamped (Math|Vector|DotProduct (Math|Vector|GetForwardVector _returnvalue_rotation) "0.000000,0.000000,1.000000") 0.0 1.0 3e+03 2e+03) (Math|Float|MapRangeClamped (Math|Vector|DotProduct (Math|Vector|GetForwardVector _returnvalue_rotation) "0.000000,0.000000,1.000000") 0.0 1.0 3e+03 2e+03) "/RocketDrillGameplay/Gameplay/Items/Curve_TriBurstVelocity.Curve_TriBurstVelocity" 4e+01 0 "MaintainLastRootMotionVelocity" "0, 0, 0" 0.0)
        (:then
          (Variables|Default|SetRootMotionTask _asynctaskproxy)
          (Variables|/Script/Engine.CharacterMovementComponent|SetGravityScale 1.0 (Variables|/Script/Engine.Character|GetCharacterMovement (Variables|Default|GetPlayerPawn))))
        (:OnFinish
          (Ability|ApplyGameplayEffectToOwner "/RocketDrillGameplay/Gameplay/Items/GameplayEffects/GE_SkidToStop.GE_SkidToStop_C")
          (Ability|RemoveGameplayCueFromOwner (Variables|GameplayCues|GetBoostingTag))
          (Ability|EndAbility))))))

(event Ability|EventActivateAbilityFromEvent
  (params (EventData :type Gameplay%20Event%20Data%20Structure%20%28by%20ref%29))
  (|CallParentFunction EventData)
  (bind _returnvalue (Ability|CommitAbilityCost false))
  (if _returnvalue
    (bind _outfortplayerpawn (CallFunction|GetAvatarasValidFortPlayerPawn)
      (:ValidFortPlayerPawn
        (Variables|Default|SetPlayerPawn _outfortplayerpawn)
        (Variables|Default|SetStructuresDestroyed 0)
        (Variables|Default|SetbWillImpact false)
        (Variables|Default|SetbPostLaunch false)
        (CallFunction|SetState "NewEnumerator0")))))

(event Custom|Update
  (bind _hitactor (Variables|Default|GetHitActor))
  (Variables|Default|SetActualVelocity (/ (- (Transformation|GetActorLocation (Variables|Default|GetPlayerPawn)) (Variables|Default|GetLastPosition)) (- (Utilities|Time|GetGameTimeinSeconds) (Variables|Default|GetLastTime))))
  (Utilities|FlowControl|Sequence
    (:then_0
      (bind _outhits (Collision|MultiSphereTraceByProfile (Transformation|GetActorLocation (Variables|Default|GetPlayerPawn)) (+ (Transformation|GetActorLocation (Variables|Default|GetPlayerPawn)) (* (Math|Vector|Normalize (Variables|Default|GetActualVelocity)) (ScalableFloat|GetValueatLevel (Variables|Default|GetSF_EnvironmentDamageTrace_Distance)))) (ScalableFloat|GetValueatLevel (Variables|Default|GetSF_EnvironmentDamageTrace_Radius)) "OverlapAll" false (Utilities|Array|MakeArray (Variables|Default|GetPlayerPawn)) "None" true "(R=1.000000,G=0.000000,B=0.000000,A=1.000000)" "(R=0.000000,G=1.000000,B=0.000000,A=1.000000)" 0.2))
      (for _array_element _outhits
        (Variables|Default|SetHitResult _array_element)
        (Variables|Default|SetHitActor (.HitActor (Collision|BreakHitResult _array_element)))
        (bind _asfort_pawn (Utilities|Casting|CastToFortPawn (Variables|Default|GetHitActor))
          (:then
            (if (not (GameplayTags|HasMatchingGameplayTag _asfort_pawn "(TagName=\"Granted.Athena.RocketDrill.HitByDrill\")"))
              (bind (_returnvalue_x _returnvalue_y) (Math|Vector|BreakVector (* (Math|Vector|Normalize (- (Transformation|GetActorLocation _asfort_pawn) (Transformation|GetActorLocation (Variables|Default|GetPlayerPawn)))) (Math|Float|SelectFloat 3e+03 2e+03 (GameplayTags|HasMatchingGameplayTag (Variables|Default|GetPlayerPawn) "(TagName=\"Granted.Athena.RocketDrill.HasWeaponVersion\")")))))
              (Character|LaunchCharacter _asfort_pawn (Math|Vector|MakeVector _returnvalue_x _returnvalue_y 1.2e+03) false false)
              (Ability|ApplyGameplayEffectSpecToTarget (Ability|MakeOutgoingGameplayEffectSpec "/RocketDrillGameplay/Gameplay/Items/GameplayEffects/GE_RocketDrill_DamageOverTime.GE_RocketDrill_DamageOverTime_C") (Ability|TargetData|AbilityTargetDatafromActor _asfort_pawn))
              (Ability|ApplyGameplayEffectToTarget (Ability|TargetData|AbilityTargetDatafromActor _asfort_pawn) "/RocketDrillGameplay/Gameplay/Items/GameplayEffects/GE_RocketDrill_WasHit.GE_RocketDrill_WasHit_C")
              (CallFunction|FireImpactEffects (Variables|Default|GetHitActor) (Variables|GameplayCues|GetPlayerImpactTag) (Variables|Default|GetHitResult))))
          (:CastFailed
            (bind _asbuilding_actor (Utilities|Casting|CastToBuildingActor (Variables|Default|GetHitActor))
              (:then
                (if (Variables|/Script/Engine.Actor|GetbCanBeDamaged _asbuilding_actor)
                  (CallFunction|FireImpactEffects (Variables|Default|GetHitActor) (Variables|GameplayCues|GetDestructibleImpactTag) (Variables|Default|GetHitResult))
                  (if (>= (Variables|Default|GetStructuresDestroyed) (ScalableFloat|GetValueasInteger (Variables|Default|GetSF_MaxStructuresDestroyed)))
                    ;[0.5 0.5 0.5 0.5] Need to add additional checks here for prop size. There don't seem to be tags for that...
                    (if (not (.bSuccess (Utilities|Casting|CastToBuildingProp _asbuilding_actor)))
                      (GameplayTasks|EndTask (Variables|Default|GetRootMotionTask))
                      (CallFunction|SetState "NewEnumerator3"))
                    (else
                      (GameplayEffects|ApplyGameplayEffectToTarget :self (Ability|GetFortAbilitySystemComponent (Variables|Default|GetPlayerPawn)) :GameplayEffectClass "/RocketDrillGameplay/Gameplay/Items/GameplayEffects/GE_RocketDrill_Damage_DirectHit_AlwaysDestroy.GE_RocketDrill_Damage_DirectHit_AlwaysDestroy_C" :Target (Ability|GetFortAbilitySystemComponent _asbuilding_actor) :Level 0.0)
                      ;[0.5 0.5 0.5 0.5] Need to add additional checks here for prop size. There don't seem to be tags for that...
                      (if (not (.bSuccess (Utilities|Casting|CastToBuildingProp _asbuilding_actor)))
                        (Math|Integer|IncrementInt (Variables|Default|GetStructuresDestroyed)))))
                  (else
                    (CallFunction|FireImpactEffects (Variables|Default|GetHitActor) (Variables|GameplayCues|GetIndestructibleImpactTag) (Variables|Default|GetHitResult)))))
              (:CastFailed
                (Utilities|Casting|CastToFortVehicleInterface _hitactor
                  (:then
                    (GameplayEffects|ApplyGameplayEffectToTarget :self (Ability|GetFortAbilitySystemComponent (Variables|Default|GetPlayerPawn)) :GameplayEffectClass "/RocketDrillGameplay/Gameplay/Items/GameplayEffects/GE_RocketDrill_VehicleDamage.GE_RocketDrill_VehicleDamage_C" :Target (Ability|GetFortAbilitySystemComponent _hitactor) :Level 0.0)
                    (Audio|BroadcastAbilityImpactatLocation 0 _hitactor (* (CallFunction|GetKnockbackDirection 0.7 0.5) 1e+03) 7.8e+02) ; This is the secret node that launches vehicles. Note it needs the 780 radius to work.
                    (CallFunction|FireImpactEffects (Variables|Default|GetHitActor) (Variables|GameplayCues|GetDestructibleImpactTag) (Variables|Default|GetHitResult)))
                  (:CastFailed
                    (CallFunction|CalculateIndestructibleEffects (Variables|Default|GetHitActor))))))))))
    (:then_1
      (if (not (Variables|Default|GetbWillImpact))
        (Variables|Default|SetbWillImpact (>= (Math|Vector|DotProduct (Math|Vector|Normalize (Variables|Default|GetActualVelocity)) "0.000000,0.000000,-1.000000") (Math|Trig|Sin(Degrees) (Variables|Default|GetAngleOfFallToTriggerImpact))))))
    (:then_2
      (Variables|Default|SetLastPosition (Transformation|GetActorLocation (Variables|Default|GetPlayerPawn)))
      (Variables|Default|SetLastTime (Utilities|Time|GetGameTimeinSeconds)))))

(event Ability|EventOnEndAbility
  (params (bWasCancelled :type Boolean))
  (bind _rootmotiontask (Variables|Default|GetRootMotionTask))
  (Ability|RemoveGameplayEffectFromOwnerWithHandle (Variables|Default|GetGE_StartupState_Handle))
  (Ability|RemoveGameplayEffectFromOwnerWithHandle (Variables|Default|GetGE_BlockActionsTraversing_Handle))
  (Ability|RemoveGameplayEffectFromOwnerWithHandle (Variables|Default|GetGE_BlockActionsRecovery_Handle))
  (Ability|RemoveGameplayEffectFromOwnerWithHandle (Variables|Default|GetGE_Skidding_Handle))
  (Ability|RemoveGameplayEffectFromOwnerWithHandle (Variables|Default|GetGE_Friction_Handle))
  (Input|SetIgnoreMoveInput (Fortnite|GetFortPlayerControllerfromActor (Variables|Default|GetPlayerPawn)) false)
  (Camera|SetBase3PCamera (Ability|GetCurrentSourceWeapon) (Variables|Default|GetBaseCameraMode))
  (Class|TestCameraModifier|WindDownandRemove (Variables|Default|GetCameraMod))
  (Utilities|Time|ClearandInvalidateTimerbyHandle (Variables|Default|GetUpdateHandle))
  (Variables|/Script/Engine.CharacterMovementComponent|SetGravityScale 1.0 (Variables|/Script/Engine.Character|GetCharacterMovement (Variables|Default|GetPlayerPawn)))
  (Utilities|IsValid _rootmotiontask
    (:Is%20Valid
      (GameplayTasks|EndTask _rootmotiontask))))

(fn CalculateIndestructibleEffects
  (params (Actor :type Actor%20Object%20Reference))
  (bind _actor (Variables|GetActor))
  (Utilities|Casting|CastToVolume Actor
    (:CastFailed
      (Utilities|Casting|CastToFortWaterBodyActor _actor
        (:then
          (CallFunction|FireImpactEffects (Variables|Default|GetHitActor) (Variables|GameplayCues|GetIndestructibleImpactTag) (Variables|Default|GetHitResult)))
        (:CastFailed
          (Utilities|Casting|CastToLandscape _actor
            (:CastFailed
              (CallFunction|FireImpactEffects (Variables|Default|GetHitActor) (Variables|GameplayCues|GetIndestructibleImpactTag) (Variables|Default|GetHitResult)))))))))

(fn FireImpactEffects
  (params (HitActor :type Actor%20Object%20Reference) (GameplayCueTag :type Gameplay%20Tag%20Structure) (HitResult :type Hit%20Result%20Structure%20%28by%20ref%29))
  (Networking|CanExecuteCosmeticEvents
    (:True
      (GameplayCue|ExecuteGameplayCueOnActor(Burst) HitActor GameplayCueTag (GameplayCue|MakeGameplayCueParametersfromHitResult HitResult)))))

(fn GetKnockbackDirection
  (params (UpAmount :type Float%20%28double-precision%29) (RightAmount :type Float%20%28single-precision%29))
  (return (Utilities|Operators|Add (Math|Vector|Normalize2D(Vector) (Math|Vector|GetForwardVector (Pawn|GetBaseAimRotation (Variables|Default|GetPlayerPawn)))) (Math|Vector|MakeVector 0.0 0.0 UpAmount) (* (Math|Vector|CrossProduct (Math|Vector|Normalize2D(Vector) (Math|Vector|GetForwardVector (Pawn|GetBaseAimRotation (Variables|Default|GetPlayerPawn)))) "0.000000,0.000000,-1.000000") RightAmount))))

(fn SetState
  (params (inState :type ERocketDrillState%20Enum :default "NewEnumerator0"))
  (bind _output_get (Variables|Default|SeteDrillState inState))
  (CallFunction|OnChangeState _output_get))
