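; Gameplay ability for the jetpack-variant Rocket Drill: ejects the player from vehicles or mounts, hops, then activates a drilling-flight movement mode extension.
; A timer-driven sphere trace damages pawns, knocks vehicles away, destroys props or buildings instantly or after delay, and bonks on head-on indestructible hits.
; Stop input respects minimum launch time; fuel or timeout also ends flight, and empty reserve removes the item.
; Ending restores movement and grants landing protection.
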(event Custom|HitMinTime
  (Ability|HasAuthority
    (:False
      (if (Variables|Default|GetPlayerTappedButton)
        (FailedAction|UnabletoPerformAction (Variables|Default|GetPlayerPawn) true (GameplayTags|MakeLiteralGameplayTagContainer "(GameplayTags=((TagName=\"Activate.Failed.RocketDrill.DidntHold\")))")) ; show reason for failing on screen if one is given
      )
      (Ability|EndAbility))))

(event Custom|CheckEject
  (bind _playerpawn (Variables|Default|GetPlayerPawn))
  (bind _returnvalue (Actor|GetComponentbyClass _playerpawn "/Script/RidingCodeRuntime.RiderComponent"))
  (bind _playerpawn_2 (Variables|Default|GetPlayerPawn))
  (bind _returnvalue_2 (Vehicle|GetVehicle _playerpawn_2))
  (bind _donotuse_activeridable (Variables|/Script/RidingCodeRuntime.RiderComponent|GetDoNotUse_ActiveRidable _returnvalue))
  (bind _destructiontraceignoredactors (Variables|Default|GetDestructionTraceIgnoredActors))
  ;[0.07432 0.3973 1] Check for riding/vehicle ejection
  (bind _playerpawn_3 (Variables|Default|GetPlayerPawn)
    (:then
      (if (GameplayTags|HasMatchingGameplayTag _playerpawn_3 "(TagName=\"Gameplay.Status.InVehicle\")")
        (Utilities|IsValid _returnvalue_2
          (:Is%20Valid
            (Utilities|Array|Add _destructiontraceignoredactors _returnvalue_2)
            (Utilities|Array|AppendArray _destructiontraceignoredactors (Vehicle|GetAllPassengers _returnvalue_2))
            (bind _returnvalue_3 (Vehicle|EjectPlayerPawn _returnvalue_2 _playerpawn_2))
            (if _returnvalue_3
              (Ability|Tasks|WaitGameplayTagRemove "(TagName=\"Gameplay.Status.InVehicle\")" 0 true
                (:Removed
                  (CallFunction|OnEjectionCheckComplete)))
              (else
                (Ability|EndAbility)))))
        (elif (GameplayTags|HasMatchingGameplayTag (Variables|Default|GetPlayerPawn) "(TagName=\"Granted.Riding.Player.IsRiding\")")
          (Utilities|IsValid _donotuse_activeridable
            (:Is%20Valid
              (Utilities|Array|Add (Variables|Default|GetDestructionTraceIgnoredActors) (Components|GetOwner _donotuse_activeridable))
              (Utilities|IsValid _returnvalue
                (:Is%20Valid
                  (bind _returnvalue_4 (Riding|StopRidingActor _returnvalue true true false true))
                  (if (or _returnvalue_4 (not (Networking|IsServer)))
                    (Ability|Tasks|WaitGameplayTagRemove "(TagName=\"Granted.Riding.Player.IsRiding\")" 0 true
                      (:Removed
                        (CallFunction|OnEjectionCheckComplete)))
                    (else
                      _)))
                (:Is%20Not%20Valid
                  _))
              (Ability|EndAbility))
            (:Is%20Not%20Valid
              _))
          (else
            (CallFunction|OnEjectionCheckComplete)))))
    (:else)))

(event Ability|EventActivateAbility
  (|CallParentFunction)
  (Variables|Default|SetWeapon (Ability|GetCurrentSourceWeapon))
  (bind _asplayer_pawn_athena (Utilities|Casting|CastToPlayerPawn_Athena (Ability|GetActivatingPawn))
    (:then
      (Variables|Default|SetPlayerPawn _asplayer_pawn_athena)
      (CallFunction|CheckEject))
    (:CastFailed
      (Ability|EndAbility))))

(event Custom|OnEjectionCheckComplete
  (bind _returnvalue (Utilities|Time|GetGameTimeinSeconds))
  (bind _abilitystartedtime (Variables|Default|GetAbilityStartedTime))
  (bind _returnvalue_2 (- _returnvalue _abilitystartedtime))
  (bind _minlaunchtime (Variables|Default|GetMinLaunchTime))
  (bind _returnvalue_3 (- _minlaunchtime _returnvalue_2))
  (Ability|Tasks|WaitNetSync "BothWait")
  (Utilities|FlowControl|Sequence
    (:then_0
      (Ability|ApplyGameplayEffectToOwner "/RocketDrillGameplay/Gameplay/Items/GameplayEffects/GE_PrimaryActive.GE_PrimaryActive_C")
      (Variables|Default|SetAbilityStartedTime (Utilities|Time|GetGameTimeinSeconds))
      (Ability|ApplyGameplayEffectToOwner "/RocketDrillGameplay/Gameplay/Items/GameplayEffects/GE_InStartupState.GE_InStartupState_C")
      (Ability|Tasks|WaitGameplayEvent "(TagName=\"Event.Athena.RocketDrill.Stop\")"
        (:EventReceived
          (if (> _returnvalue_2 _minlaunchtime)
            (Ability|EndAbility)
            (else
              (Variables|Default|SetPlayerTappedButton (> _returnvalue_3 (* (Variables|Default|GetMinLaunchTime) 0.5)))
              (Utilities|Time|SetTimerbyEvent (delegate Custom|HitMinTime) _returnvalue_3)
              (if (Variables|Default|GetPlayerTappedButton)
                (FailedAction|UnabletoPerformAction (Variables|Default|GetPlayerPawn) true (GameplayTags|MakeLiteralGameplayTagContainer "(GameplayTags=((TagName=\"Activate.Failed.RocketDrill.DidntHold\")))")) ; show reason for failing on screen if one is given
              ))))))
    (:then_1
      (bind _asfort_movement_comp_character (Utilities|Casting|CastToFortMovementComp_Character (Variables|/Script/Engine.Character|GetCharacterMovement (Variables|Default|GetPlayerPawn))))
      (NetTolerance|SetUseHighAccelerationTolerance _asfort_movement_comp_character true "RocketDrillMovement") ; TODO this is to temporarily hide net corrections, will need to be removed eventually
      (Pawn|Components|CharacterMovement|SetWalkableFloorAngle (Variables|/Script/Engine.Character|GetCharacterMovement (Variables|Default|GetPlayerPawn)) (ScalableFloat|GetValueatLevel (Variables|ScalableFloats|GetSF_NewWalkableAngle)))
      (Ability|ExecuteGameplayCueOnOwner (Variables|GCNs|GetStartupCue))
      (Character|LaunchCharacter (Variables|Default|GetPlayerPawn) (Math|Vector|MakeVector 0.0 0.0 (ScalableFloat|GetValueatLevel (Variables|ScalableFloats|GetSF_InitialHopZVelocity))) false true)
      (Utilities|FlowControl|Delay 0.28)
      (Ability|AddGameplayCueToOwner (Variables|GCNs|GetLoopingTraversalCue))
      (Ability|RemoveGameplayEffectFromOwnerWithGrantedTags "(GameplayTags=((TagName=\"Granted.Athena.RocketDrill.Startup\")))")
      (CallFunction|SpawnImpactActor)
      (Class|FortMovementModeExtBlueprintLibrary|ActivateRegisteredMovementModeExtension "/RocketDrillGameplay/Gameplay/Items/MME/MME_Logic_RocketDrill_Native_JetpackVer.MME_Logic_RocketDrill_Native_JetpackVer_C" (Variables|Default|GetPlayerPawn))
      (Utilities|FlowControl|Delay 0.2)
      (Variables|Default|SetWaitingToLaunch false))
    (:then_2
      (CallFunction|StartTimers))
    (:then_3
      (if (and (ScalableFloat|GetValueasBool (Variables|ScalableFloats|GetSF%20DelayWaterCollisionBlockUntilSurfaced)) (GameplayTags|HasMatchingGameplayTag (Ability|GetFortAbilitySystemComponent (Variables|Default|GetPlayerPawn)) (Variables|Default|GetUnderwaterTag)))
        (Ability|Tasks|WaitGameplayTagRemove (Variables|Default|GetUnderwaterTag)
          (:Removed
            (Collision|SetCollisionResponsetoChannel (Variables|/Script/Engine.Character|GetCapsuleComponent (Variables|Default|GetPlayerPawn)) "ECC_GameTraceChannel18" "ECR_Block")))
        (else
          (Collision|SetCollisionResponsetoChannel (Variables|/Script/Engine.Character|GetCapsuleComponent (Variables|Default|GetPlayerPawn)) "ECC_GameTraceChannel18" "ECR_Block"))))))

(event Custom|StartTimers
  (Variables|Default|SetLastPosition (Transformation|GetWorldLocation (Variables|/Script/Engine.Character|GetCapsuleComponent (Variables|Default|GetPlayerPawn))))
  (bind _returnvalue (Utilities|Time|SetTimerbyEvent (delegate Custom|DoUpdate) 0.05 true true))
  (Variables|Default|SetUpdateTimer _returnvalue)
  (Utilities|Time|SetTimerbyEvent (delegate Custom|PayDurabilityCost) 0.4 true)
  (Utilities|FlowControl|Delay 1.25)
  (Utilities|Time|SetTimerbyEvent (delegate Custom|TimeOut) 23.0))

(event Custom|TimeOut
  (Ability|AddGameplayCueToOwner (Variables|GCNs|GetTimeoutCue))
  (Utilities|FlowControl|Delay 4.0)
  (Ability|ApplyGameplayEffectToOwner "/RocketDrillGameplay/Gameplay/Items/GameplayEffects/GE_TimedOut.GE_TimedOut_C")
  (Ability|EndAbility))

(event Custom|PayDurabilityCost
  (bind (_bsuccessfullyfoundattribute _returnvalue) (Ability|Attribute|GetFloatAttribute (Variables|Default|GetPlayerPawn) "(AttributeName=\"ReserveFuel\",Attribute=/Script/RocketDrillGameplayRuntime.RocketDrillAttributeSet:ReserveFuel,AttributeOwner=\"/Script/CoreUObject.Class'/Script/RocketDrillGameplayRuntime.RocketDrillAttributeSet'\")"))
  (bind _returnvalue_2 (Utilities|Time|GetGameTimeinSeconds))
  (bind _abilitystartedtime (Variables|Default|GetAbilityStartedTime))
  (bind _returnvalue_3 (- _returnvalue_2 _abilitystartedtime))
  (bind _outoffuelminlaunchtime (Variables|Default|GetOutOfFuelMinLaunchTime))
  (if (not (ScalableFloat|GetValueasBoolwithContextParam (Variables|Default|GetPlayerPawn) (Variables|/Script/FortniteGame.FortExperienceSettingsComponent|GetInfiniteReserveEnergy (Fortnite|Athena|GameState|GetExperienceSettingsComponentfromGameState))))
    (Ability|CommitAbilityCost false)
    (if (<= _returnvalue 0)
      (if (> _returnvalue_3 _outoffuelminlaunchtime)
        (Inventory|RemoveFortItemFromPlayer (Fortnite|GetFortPlayerControllerfromActor (Variables|Default|GetPlayerPawn)) (FortWeapon|GetInventoryItem (Ability|GetCurrentSourceWeapon)) -1 true)
        (Ability|ExecuteGameplayCueOnOwner (Variables|GCNs|GetEquippedItemDestroyed))
        (Ability|EndAbility)
        (else
          (Utilities|Time|SetTimerbyEvent (delegate Custom|TotalFuelEmpty) (- _outoffuelminlaunchtime _returnvalue_3)))))))

(event Custom|DoUpdate
  (CallFunction|Update false))

(event Custom|TotalFuelEmpty
  (Inventory|RemoveFortItemFromPlayer (Fortnite|GetFortPlayerControllerfromActor (Variables|Default|GetPlayerPawn)) (FortWeapon|GetInventoryItem (Ability|GetCurrentSourceWeapon)) -1 true)
  (Ability|ExecuteGameplayCueOnOwner (Variables|GCNs|GetEquippedItemDestroyed))
  (Ability|EndAbility))

(event Custom|Update
  (params (FinalUpdate :type Boolean))
  (bind _playerpawn (Variables|Default|GetPlayerPawn))
  (bind _returnvalue (Ability|GetFortAbilitySystemComponent _playerpawn))
  (bind _returnvalue_2 (Utilities|Time|GetGameTimeinSeconds))
  (bind _playerpawn_2 (Variables|Default|GetPlayerPawn))
  (bind _returnvalue_3 (Ability|GetFortAbilitySystemComponent _playerpawn_2))
  (bind (_bsuccessfullyfoundattribute _returnvalue_4) (Ability|Attribute|GetFloatAttribute (Variables|Default|GetPlayerPawn) "(AttributeName=\"Fuel\",Attribute=/Script/RocketDrillGameplayRuntime.RocketDrillAttributeSet:Fuel,AttributeOwner=\"/Script/CoreUObject.Class'/Script/RocketDrillGameplayRuntime.RocketDrillAttributeSet'\")"))
  (bind _abilitystartedtime (Variables|Default|GetAbilityStartedTime))
  (bind _returnvalue_5 (- _returnvalue_2 _abilitystartedtime))
  (bind _outoffuelminlaunchtime (Variables|Default|GetOutOfFuelMinLaunchTime))
  (bind _currenthitactor (Variables|Audio|GetCurrentHitActor))
  (if (not (Variables|Default|GetWaitingToLaunch))
    (Variables|Default|SetActualVelocity (/ (- (Transformation|GetWorldLocation (Variables|/Script/Engine.Character|GetCapsuleComponent (Variables|Default|GetPlayerPawn))) (Variables|Default|GetLastPosition)) (Utilities|Time|GetWorldDeltaSeconds)))
    (Utilities|FlowControl|Sequence
      (:then_1
        (bind _outhits (Collision|MultiSphereTraceByProfile (+ (Transformation|GetWorldLocation (Variables|/Script/Engine.Character|GetCapsuleComponent (Variables|Default|GetPlayerPawn))) "0.000000,0.000000,10.000000") (+ (+ (Transformation|GetWorldLocation (Variables|/Script/Engine.Character|GetCapsuleComponent (Variables|Default|GetPlayerPawn))) "0.000000,0.000000,10.000000") (* (Math|Vector|Normalize (Variables|Default|GetActualVelocity)) (Math|Float|SelectFloat 3.5e+02 (ScalableFloat|GetValueatLevel (Variables|ScalableFloats|GetSF_EnvironmentDamageTrace_Distance)) FinalUpdate))) (ScalableFloat|GetValueatLevel (Variables|ScalableFloats|GetSF_EnvironmentDamageTrace_Radius)) "OverlapAll" false (Utilities|Array|MakeArray (Variables|Default|GetPlayerPawn)) "None" true "(R=1.000000,G=0.000000,B=0.000000,A=1.000000)" "(R=0.000000,G=1.000000,B=0.000000,A=1.000000)" 0.2))
        (Audio|TracetoDrillAudio _outhits)
        (Variables|Default|SetHitAnyBuildingActors false)
        (for _array_element _outhits
          (Variables|Audio|SetCurrentHitResult _array_element)
          (bind _output_get (Variables|Audio|SetCurrentHitActor (.HitActor (Collision|BreakHitResult _array_element))))
          (bind _asfort_pawn (Utilities|Casting|CastToFortPawn (Variables|Audio|GetCurrentHitActor))
            (:then
              (if (not (GameplayTags|HasMatchingGameplayTag _asfort_pawn "(TagName=\"Granted.Athena.RocketDrill.HitByDrill\")"))
                (bind _returnvalue_6 (Fortnite|CheckLineOfSightToActorWithChannel (Transformation|GetActorLocation (Variables|Default|GetPlayerPawn)) _asfort_pawn "ECC_GameTraceChannel7" (Variables|Default|GetPlayerPawn)))
                (if _returnvalue_6
                  (Ability|ApplyGameplayEffectToTarget (Ability|TargetData|AbilityTargetDatafromActor _asfort_pawn) "/RocketDrillGameplay/Gameplay/Items/GameplayEffects/GE_GrantKnockbackandFallImmunity.GE_GrantKnockbackandFallImmunity_C")
                  (bind _returnvalue_7 (Ability|GameplayEffect|AssignTagSetbyCallerMagnitude (Ability|MakeOutgoingGameplayEffectSpec "/RocketDrillGameplay/Gameplay/Items/GameplayEffects/GE_RocketDrill_Damage_DirectHit_AlwaysDestroy.GE_RocketDrill_Damage_DirectHit_AlwaysDestroy_C") "(TagName=\"SetByCaller.AbilityDamage\")" (ScalableFloat|GetValueatLevel (Variables|ScalableFloats|GetSF_ImpactPlayerDamage))))
                  (Ability|ApplyGameplayEffectSpecToTarget _returnvalue_7 (Ability|TargetData|AbilityTargetDatafromActor _asfort_pawn))
                  (Ability|ApplyGameplayEffectToTarget (Ability|TargetData|AbilityTargetDatafromActor _asfort_pawn) "/RocketDrillGameplay/Gameplay/Items/GameplayEffects/GE_RocketDrill_WasHit.GE_RocketDrill_WasHit_C")
                  (Audio|FireImpactEffects (Variables|Audio|GetCurrentHitActor) (Variables|GCNs|GetPlayerImpactTag) (Variables|Audio|GetCurrentHitResult)))))
            (:CastFailed
              (bind _asbuilding_actor (Utilities|Casting|CastToBuildingActor _output_get)
                (:then
                  (if (Utilities|IsPhysicsEnabled (.HitComponent (Collision|BreakHitResult _array_element)))
                    (bind _returnvalue_8 (Class|BuildingActor|DoesMutatorBlockIncomingDamagewithTags _asbuilding_actor (Variables|Default|GetMutatorBlocksWeaponsQuery)))
                    (if (and (not _returnvalue_8) (Variables|/Script/Engine.Actor|GetbCanBeDamaged _asbuilding_actor))
                      (bind (_origin _boxextent) (Collision|GetActorBounds _asbuilding_actor true false))
                      (bind (_boxextent_x _boxextent_y _boxextent_z) (Math|Vector|BreakVector _boxextent))
                      (if (and (.bSuccess (Utilities|Casting|CastToBuildingProp _asbuilding_actor)) (< (Utilities|Operators|Multiply _boxextent_x _boxextent_y _boxextent_z 0.0001) (ScalableFloat|GetValueatLevel (Variables|ScalableFloats|GetSF_MaxPropBoundsInstantDestruction))))
                        (GameplayEffects|ApplyGameplayEffectToTarget :self _returnvalue_3 :GameplayEffectClass "/RocketDrillGameplay/Gameplay/Items/GameplayEffects/GE_RocketDrill_Damage_DirectHit_AlwaysDestroy.GE_RocketDrill_Damage_DirectHit_AlwaysDestroy_C" :Target (Ability|GetFortAbilitySystemComponent _asbuilding_actor) :Level 0.0)
                        (Audio|FireImpactEffects (Variables|Audio|GetCurrentHitActor) (Variables|GCNs|GetDestructibleImpactTag) (Variables|Audio|GetCurrentHitResult))
                        (else
                          (GameplayEffects|ApplyGameplayEffectToTarget :self _returnvalue_3 :GameplayEffectClass "/RocketDrillGameplay/Gameplay/Items/GameplayEffects/GE_RocketDrill_DamageOverTime.GE_RocketDrill_DamageOverTime_C" :Target (Ability|GetFortAbilitySystemComponent _asbuilding_actor) :Level 0.0)
                          (Variables|Default|SetHitAnyBuildingActors true)
                          (Utilities|Array|Add (Variables|Default|GetHitBuildingActors) _asbuilding_actor)))
                      (elif (CallFunction|CheckCollisionAngle _array_element)
                        (Audio|FireImpactEffects (Variables|Audio|GetCurrentHitActor) (Variables|GCNs|GetIndestructibleImpactTag) (Variables|Audio|GetCurrentHitResult))))))
                (:CastFailed
                  (Utilities|Casting|CastToFortVehicleInterface _currenthitactor
                    (:then
                      (if (not (GameplayTags|HasMatchingGameplayTag (Ability|GetFortAbilitySystemComponent _currenthitactor) "(TagName=\"Granted.Athena.RocketDrill.HitByDrill\")"))
                        (GameplayEffects|ApplyGameplayEffectToTarget :self _returnvalue :GameplayEffectClass "/RocketDrillGameplay/Gameplay/Items/GameplayEffects/GE_RocketDrill_VehicleDamage.GE_RocketDrill_VehicleDamage_C" :Target (Ability|GetFortAbilitySystemComponent _currenthitactor) :Level 0.0)
                        (GameplayEffects|ApplyGameplayEffectToTarget :self _returnvalue :GameplayEffectClass "/RocketDrillGameplay/Gameplay/Items/GameplayEffects/GE_RocketDrill_WasHit.GE_RocketDrill_WasHit_C" :Target (Ability|GetFortAbilitySystemComponent _currenthitactor) :Level 0.0)
                        (Audio|BroadcastAbilityImpactatLocation 0 _currenthitactor (* (CallFunction|GetKnockbackDirection 0.7 0.5) 1e+03) 7.8e+02) ; This is the secret node that launches vehicles. Note it needs the 780 radius to work.
                      ))
                    (:CastFailed
                      (if (Utilities|IsA _currenthitactor "/Script/FortniteGame.FortWaterBodyCubeVolume")
                        _
                        (elif (Utilities|Enum|Equal(Enum) (Collision|GetCollisionResponsetoChannel (.HitComponent (Collision|BreakHitResult _array_element)) "ECC_GameTraceChannel6") "ECR_Block")
                          (Audio|FireImpactEffects (Variables|Audio|GetCurrentHitActor) (Variables|GCNs|GetIndestructibleImpactTag) (Variables|Audio|GetCurrentHitResult))
                          (if (CallFunction|CheckCollisionAngle _array_element)
                            _)))))))
              (Ability|ApplyGameplayEffectToOwner "/RocketDrillGameplay/Gameplay/Items/GameplayEffects/GE_Bonk_HoldPosition.GE_Bonk_HoldPosition_C")
              (Ability|ApplyGameplayEffectToOwner "/RocketDrillGameplay/Gameplay/Items/GameplayEffects/GE_LandingImpact.GE_LandingImpact_C")
              (bind _output_get_2 (Variables|Default|SetHardImpactContext (Class|AthenaFunctionLibrary|MakeSimpleFortEffectContext (Variables|Default|GetPlayerPawn) (Variables|Default|GetPlayerPawn))))
              (Ability|EffectContext|AddHitResult _output_get_2 (Variables|Audio|GetCurrentHitResult))
              (Ability|ExecuteGameplayCueWithParamsOnOwner (Variables|GCNs|GetBonkCue) (Ability|GameplayCue|MakeGameplayCueParameters 0.0 0.0 (Variables|Default|GetHardImpactContext)))
              (Ability|EndAbility))))
        (if (Variables|Default|GetHitAnyBuildingActors)
          (Ability|ApplyGameplayEffectToOwner "/RocketDrillGameplay/Gameplay/Items/GameplayEffects/GE_RocketDrill_BlockedByObstacle.GE_RocketDrill_BlockedByObstacle_C")
          (if (not (Utilities|Time|IsValidTimerHandle (Variables|Default|GetPendingDestroy)))
            (bind _returnvalue_9 (Utilities|Time|SetTimerbyEvent (delegate Custom|DestroyPendingBuildingActors) (ScalableFloat|GetValueatLevel (Variables|ScalableFloats|GetSF_TimeToDestroyBuild))))
            (Variables|Default|SetPendingDestroy _returnvalue_9))
          (else
            (Variables|Default|SetDamageProgress 0))))
      (:then_2
        (Variables|Default|SetLastPosition (Transformation|GetWorldLocation (Variables|/Script/Engine.Character|GetCapsuleComponent (Variables|Default|GetPlayerPawn))))
        (if (<= _returnvalue_4 0)
          (if (> _returnvalue_5 _outoffuelminlaunchtime)
            (Ability|EndAbility)
            (else
              (Utilities|Time|SetTimerbyEvent (delegate Custom|OutOfFuelMinTime) (- _outoffuelminlaunchtime _returnvalue_5)))))))))

(event Custom|OutOfFuelMinTime
  (Ability|EndAbility))

(event Custom|DestroyPendingBuildingActors
  (Utilities|Time|ClearandInvalidateTimerbyHandle (Variables|Default|GetPendingDestroy))
  (for _array_element (Variables|Default|GetHitBuildingActors)
    (GameplayEffects|ApplyGameplayEffectToTarget :self (Ability|GetFortAbilitySystemComponent (Variables|Default|GetPlayerPawn)) :GameplayEffectClass "/RocketDrillGameplay/Gameplay/Items/GameplayEffects/GE_RocketDrill_Damage_DirectHit_AlwaysDestroy.GE_RocketDrill_Damage_DirectHit_AlwaysDestroy_C" :Target (Ability|GetFortAbilitySystemComponent _array_element) :Level 0.0))
  (Utilities|Array|Clear (Variables|Default|GetHitBuildingActors)))

(event Ability|EventOnEndAbility
  (params (bWasCancelled :type Boolean))
  (bind (_bblockinghit _bwalkablefloor) (Utilities|Struct|BreakFindFloorResult (Pawn|Components|CharacterMovement|FindFloor (Variables|/Script/Engine.Character|GetCharacterMovement (Variables|Default|GetPlayerPawn)) (Transformation|GetWorldLocation (Variables|/Script/Engine.Character|GetCapsuleComponent (Variables|Default|GetPlayerPawn))))))
  (bind (_bsuccessfullyfoundattribute _returnvalue) (Ability|Attribute|GetFloatAttribute (Variables|Default|GetPlayerPawn) "(AttributeName=\"ReserveFuel\",Attribute=/Script/RocketDrillGameplayRuntime.RocketDrillAttributeSet:ReserveFuel,AttributeOwner=\"/Script/CoreUObject.Class'/Script/RocketDrillGameplayRuntime.RocketDrillAttributeSet'\")"))
  (bind (_returnvalue_x _returnvalue_y _returnvalue_z) (Math|Vector|BreakVector (* (Variables|/Script/Engine.MovementComponent|GetVelocity (Variables|/Script/Engine.Character|GetCharacterMovement (Variables|Default|GetPlayerPawn))) 1.1)))
  (bind _output_get (Variables|Default|SetLaunchWasCancelled bWasCancelled))
  (|CallParentFunction _output_get)
  (bind _drillimpactactor (Variables|Audio|GetDrillImpactActor)
    (:then
      (Actor|DestroyActor _drillimpactactor))
    (:else
      _))
  (Ability|RemoveGameplayEffectFromOwnerWithGrantedTags "(GameplayTags=((TagName=\"Gameplay.Action.Player.RocketDrill.Active\")))")
  (Utilities|FlowControl|Sequence
    (:then_0
      (Pawn|Components|CharacterMovement|SetWalkableFloorAngle (Variables|/Script/Engine.Character|GetCharacterMovement (Variables|Default|GetPlayerPawn)) 44.765083)
      (Transformation|SetRelativeLocationAndRotation (Variables|/Script/Engine.Character|GetMesh (Variables|Default|GetPlayerPawn)) "0.000000,0.000000,-70.000000" "0.000000,-90.000000,0.000000" false true)
      (Development|PrintString "ON END ABILITY" true true "(R=1.000000,G=0.047253,B=0.212893,A=1.000000)")
      (bind _asfort_movement_comp_character (Utilities|Casting|CastToFortMovementComp_Character (Variables|/Script/Engine.Character|GetCharacterMovement (Variables|Default|GetPlayerPawn))))
      (NetTolerance|SetUseHighAccelerationTolerance _asfort_movement_comp_character false "RocketDrillMovement") ; TODO this is to temporarily hide net corrections, will need to be removed eventually
      (Utilities|Time|ClearandInvalidateTimerbyHandle (Variables|Default|GetUpdateTimer))
      (Collision|SetCollisionResponsetoChannel (Variables|/Script/Engine.Character|GetCapsuleComponent (Variables|Default|GetPlayerPawn)) "ECC_GameTraceChannel18" "ECR_Overlap")
      (Ability|RemoveGameplayCueFromOwner (Variables|GCNs|GetTimeoutCue))
      (Ability|RemoveGameplayCueFromOwner (Variables|GCNs|GetLoopingTraversalCue))
      (if (<= _returnvalue 0)
        (Inventory|RemoveFortItemFromPlayer (Fortnite|GetFortPlayerControllerfromActor (Variables|Default|GetPlayerPawn)) (FortWeapon|GetInventoryItem (Ability|GetCurrentSourceWeapon)) -1 true)
        (Ability|ExecuteGameplayCueOnOwner (Variables|GCNs|GetEquippedItemDestroyed))))
    (:then_1
      (CallFunction|Update true)
      (CallFunction|DestroyPendingBuildingActors)
      (Utilities|FlowControl|Sequence
        (:then_0
          (if (and (not (Ability|EffectContext|IsValid (Variables|Default|GetHardImpactContext))) (not (Variables|Default|GetLaunchWasCancelled)))
            (Character|LaunchCharacter (Variables|Default|GetPlayerPawn) (Math|Vector|MakeVector _returnvalue_x _returnvalue_y (+ _returnvalue_z 4e+02)) true true)))
        (:then_1
          (Ability|ApplyGameplayEffectToOwner "/RocketDrillGameplay/Gameplay/Items/GameplayEffects/Holstering/GE_TillLandHolster_Grant_RocketDrill.GE_TillLandHolster_Grant_RocketDrill_C")
          (Ability|ApplyGameplayEffectToOwner "/RocketDrillGameplay/Gameplay/Items/GameplayEffects/FallImmunity/GE_RocketDrill_TillLandFallDamageImmunity_NoFX_Grant_AndMoreAirControl.GE_RocketDrill_TillLandFallDamageImmunity_NoFX_Grant_AndMoreAirControl_C")
          (Ability|ApplyGameplayEffectToOwner (select (GameplayTags|HasMatchingGameplayTag (Variables|Default|GetPlayerPawn) "(TagName=\"Granted.Athena.RocketDrill.UsingJetpackBurstVersion\")") "/RocketDrillGameplay/Gameplay/Items/GameplayEffects/GE_JetpackBurstVer_BlockRefuelingInitially_RocketDrill.GE_JetpackBurstVer_BlockRefuelingInitially_RocketDrill_C" "/RocketDrillGameplay/Gameplay/Items/GameplayEffects/GE_JetpackVer_BlockRefuelingInitially_RocketDrill.GE_JetpackVer_BlockRefuelingInitially_RocketDrill_C"))
          (if _bwalkablefloor
            (if (not (GameplayTags|HasMatchingGameplayTag (Variables|Default|GetPlayerPawn) "(TagName=\"Granted.Athena.RocketDrill.LandingWithImpact\")"))
              (Ability|ApplyGameplayEffectToOwner "/RocketDrillGameplay/Gameplay/Items/GameplayEffects/GE_SkidToStop.GE_SkidToStop_C"))))))))

(fn CheckCollisionAngle
  (params (HitResult :type Hit%20Result%20Structure))
  (bind (_bblockinghit _binitialoverlap _time _distance _location _impactpoint _normal) (Collision|BreakHitResult HitResult))
  (bind _actualvelocity (Variables|Default|GetActualVelocity))
  (bind _returnvalue (Math|Vector|Normalize _actualvelocity))
  (bind _returnvalue_2 (* _returnvalue -1.0))
  (bind _returnvalue_3 (Math|Vector|DotProduct _normal _returnvalue_2))
  (Development|PrintString (Utilities|String|ToString(Float) _returnvalue_3))
  (return (> _returnvalue_3 (Math|Trig|Cos(Degrees) (ScalableFloat|GetValueatLevel (Variables|ScalableFloats|GetSF_HardImpactCollisionAngle))))))

(fn FireImpactEffects
  (params (HitActor :type Actor%20Object%20Reference) (GameplayCueTag :type Gameplay%20Tag%20Structure) (HitResult :type Hit%20Result%20Structure%20%28by%20ref%29))
  (GameplayCue|ExecuteGameplayCueOnActor(Burst) HitActor GameplayCueTag (GameplayCue|MakeGameplayCueParametersfromHitResult HitResult)))

(fn GetKnockbackDirection
  (params (UpAmount :type Float%20%28double-precision%29) (RightAmount :type Float%20%28single-precision%29))
  (return (Utilities|Operators|Add (Math|Vector|Normalize2D(Vector) (Math|Vector|GetForwardVector (Pawn|GetBaseAimRotation (Variables|Default|GetPlayerPawn)))) (Math|Vector|MakeVector 0.0 0.0 UpAmount) (* (Math|Vector|CrossProduct (Math|Vector|Normalize2D(Vector) (Math|Vector|GetForwardVector (Pawn|GetBaseAimRotation (Variables|Default|GetPlayerPawn)))) "0.000000,0.000000,-1.000000") RightAmount))))

(fn CanActivateAbility
  (params (ActorInfo_OwnerActor :type Actor%20Object%20Reference) (ActorInfo_AvatarActor :type Actor%20Object%20Reference) (ActorInfo_PlayerController :type Player%20Controller%20Object%20Reference) (ActorInfo_AbilitySystemComponent :type Ability%20System%20Component%20Object%20Reference) (ActorInfo_SkeletalMeshComponent :type Skeletal%20Mesh%20Component%20Object%20Reference) (ActorInfo_AnimInstance :type Anim%20Instance%20Object%20Reference) (ActorInfo_MovementComponent :type Movement%20Component%20Object%20Reference) (ActorInfo_AffectedAnimInstanceTag :type Name) (Handle :type Gameplay%20Ability%20Spec%20Handle%20Structure))
  (bind (_bsuccessfullyfoundattribute _returnvalue) (Ability|Attribute|GetFloatAttribute ActorInfo_AvatarActor "(AttributeName=\"Fuel\",Attribute=/Script/RocketDrillGameplayRuntime.RocketDrillAttributeSet:Fuel,AttributeOwner=\"/Script/CoreUObject.Class'/Script/RocketDrillGameplayRuntime.RocketDrillAttributeSet'\")"))
  (bind _returnvalue_2 (> _returnvalue 0))
  (Utilities|IsValid ActorInfo_OwnerActor
    (:Is%20Valid
      (Utilities|IsValid ActorInfo_AvatarActor
        (:Is%20Valid
          (if (not _returnvalue_2)
            (GameplayCue|ExecuteGameplayCue ActorInfo_OwnerActor (Variables|GCNs|GetOutOfFuelCue)))
          (return (GameplayTags|MakeLiteralGameplayTagContainer "(GameplayTags=((TagName=\"Activate.Failed.RocketDrill.OutOfFuel\"),(TagName=\"Activate.Failed.Cooldown\")))") _returnvalue_2))
        (:Is%20Not%20Valid
          _)))
    (:Is%20Not%20Valid
      _))
  (return "(GameplayTags=((TagName=\"Activate.Failed.Cooldown\")))" false))

(fn SpawnImpactActor
  (Networking|CanExecuteCosmeticEvents
    (:False
      (bind _weapon (Variables|Default|GetWeapon)
        (:then
          (bind _returnvalue (Game|SpawnActorfromClass "/RocketDrillGameplay/Assets/Audio/BP_RocketDrill_ImpactLoop.BP_RocketDrill_ImpactLoop_C" (Transformation|GetActorTransform _weapon) "Undefined" "MultiplyWithRoot" 0 (Variables|Default|GetPlayerPawn)))
          (bind _output_get (Variables|Audio|SetDrillImpactActor _returnvalue))
          (Transformation|AttachActorToActor _output_get (Variables|Default|GetWeapon) "None" "SnapToTarget" "SnapToTarget" "SnapToTarget" true))
        (:else)))))

(fn TraceToDrillAudio
  (params (TraceResults :type Array%20of%20Hit%20Result%20Structures))
  (Networking|CanExecuteCosmeticEvents
    (:False
      (Variables|Audio|GetDrillImpactActor
        (:then
          _)
        (:else
          (bind _weapon (Variables|Default|GetWeapon)
            (:then
              (bind _output_get (LocalVariables|SetAttachedActors (Actor|GetAttachedActors _weapon true false)))
              (for _array_element _output_get
                (bind _asbp_rocket_drill_impact_loop (Utilities|Casting|CastToBP_RocketDrill_ImpactLoop _array_element))
                (Variables|Audio|SetDrillImpactActor _asbp_rocket_drill_impact_loop)
                (Class|BPRocketDrillImpactLoop|UpdatefromDamageTrace (Variables|Audio|GetDrillImpactActor) (Variables|GetTraceResults))
                (return)))
            (:else)))))))
