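; Gameplay ability for the traversal Rocket Drill: consumes ammo, then after a startup delay activates the drill movement mode extension, camera modifier and boost event.
; A repeating Update timer sphere-traces along measured velocity to damage and knock back pawns, buildings and vehicles, flag wall blocks, and signal upcoming slope peaks on non-authority.
; A stop event or timeout ends it, restoring movement and collision, launching the player with fall immunity and committing cooldown.
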(event Ability|EventActivateAbility
  (|CallParentFunction)
  (Utilities|FlowControl|Sequence
    (:then_0
      (Variables|Default|SetWeapon (Ability|GetCurrentSourceWeapon))
      (bind _asplayer_pawn_athena (Utilities|Casting|CastToPlayerPawn_Athena (Ability|GetActivatingPawn))
        (:then
          (Variables|Default|SetPlayerPawn _asplayer_pawn_athena)
          (FortWeapon|Ammo|ConsumeAmmo (Variables|Default|GetWeapon) 1 false)
          (Ability|ApplyGameplayEffectToOwner "/RocketDrillGameplay/Gameplay/Items/GameplayEffects/GE_InStartupState.GE_InStartupState_C")
          (Ability|Tasks|WaitGameplayEvent "(TagName=\"Event.Athena.RocketDrill.Stop\")"
            (:EventReceived
              (Ability|EndAbility))))
        (:CastFailed
          (Ability|EndAbility))))
    (:then_1
      (bind _asfort_movement_comp_character (Utilities|Casting|CastToFortMovementComp_Character (Variables|/Script/Engine.Character|GetCharacterMovement (Variables|Default|GetPlayerPawn))))
      (NetTolerance|SetUseHighAccelerationTolerance _asfort_movement_comp_character true "RocketDrillMovement") ; TODO this is to temporarily hide net corrections, will need to be removed eventually
      (Pawn|Components|CharacterMovement|SetWalkableFloorAngle (Variables|/Script/Engine.Character|GetCharacterMovement (Variables|Default|GetPlayerPawn)) (ScalableFloat|GetValueatLevel (Variables|Default|GetSF_NewWalkableAngle)))
      (Ability|ExecuteGameplayCueOnOwner (Variables|Default|GetStartupCue))
      (Utilities|FlowControl|Delay 2.0)
      (Ability|AddGameplayCueToOwner (Variables|Default|GetBoostingCue))
      (Class|FortMovementModeExtBlueprintLibrary|ActivateRegisteredMovementModeExtension "/RocketDrillGameplay/Gameplay/Items/MME/MME_Logic_RocketDrill.MME_Logic_RocketDrill_C" (Variables|Default|GetPlayerPawn))
      (bind _returnvalue (Game|Player|AddNewCameraModifier (Variables|/Script/Engine.PlayerController|GetPlayerCameraManager (Fortnite|GetFortPlayerControllerfromActor (Variables|Default|GetPlayerPawn))) "/RocketDrillGameplay/Gameplay/Items/CameraModifier_RocketDrill_TraversalVersion.CameraModifier_RocketDrill_TraversalVersion_C"))
      (bind _output_get (Variables|Default|SetCameraMod _returnvalue))
      (CameraModifier|EnableModifier _output_get)
      (Utilities|FlowControl|Delay 0.2)
      (Ability|SendGameplayEvent "(TagName=\"Event.Athena.RocketDrill.TriggerBoostAbility\")"))
    (:then_2
      (CallFunction|StartTimers))
    (:then_3
      (if (and (ScalableFloat|GetValueasBool (Variables|Default|GetSF%20DelayWaterCollisionBlockUntilSurfaced)) (GameplayTags|HasMatchingGameplayTag (Ability|GetFortAbilitySystemComponent (Variables|Default|GetPlayerPawn)) (Variables|Default|GetUnderwaterTag)))
        (Ability|Tasks|WaitGameplayTagRemove (Variables|Default|GetUnderwaterTag)
          (:Removed
            (Collision|SetCollisionResponsetoChannel (Variables|/Script/Engine.Character|GetCapsuleComponent (Variables|Default|GetPlayerPawn)) "ECC_GameTraceChannel18" "ECR_Block")))
        (else
          (Collision|SetCollisionResponsetoChannel (Variables|/Script/Engine.Character|GetCapsuleComponent (Variables|Default|GetPlayerPawn)) "ECC_GameTraceChannel18" "ECR_Block"))))
    (:then_4
      (Utilities|Time|SetTimerbyEvent 0 0.05 true true))))

(event Custom|StartTimers
  (bind _returnvalue (Utilities|Time|SetTimerbyEvent (EventDispatchers|CreateEvent :!event Custom|Update) 0.05 true true))
  (Variables|Default|SetSlopeCheckTimer _returnvalue)
  (Utilities|FlowControl|Delay 1.25)
  (Utilities|Time|SetTimerbyEvent (delegate Custom|TimeOut) 23.0))

(event Custom|TimeOut
  (Ability|AddGameplayCueToOwner (Variables|Default|GetTimeoutCue))
  (Utilities|FlowControl|Delay 4.0)
  (Ability|ApplyGameplayEffectToOwner "/RocketDrillGameplay/Gameplay/Items/GameplayEffects/GE_TimedOut.GE_TimedOut_C")
  (Ability|EndAbility))

(event Custom|Update
  (bind _playerpawn (Variables|Default|GetPlayerPawn))
  (bind _capsulecomponent (Variables|/Script/Engine.Character|GetCapsuleComponent _playerpawn))
  (bind _returnvalue (Transformation|GetWorldLocation _capsulecomponent))
  (bind _playerpawn_2 (Variables|Default|GetPlayerPawn))
  (bind _charactermovement (Variables|/Script/Engine.Character|GetCharacterMovement _playerpawn_2))
  (bind _floorresult (Pawn|Components|CharacterMovement|FindFloor _charactermovement _returnvalue))
  (bind (_bblockinghit _bwalkablefloor _blinetrace _floordist _linedist _hitresult) (Utilities|Struct|BreakFindFloorResult _floorresult))
  (bind (_bblockinghit_2 _binitialoverlap _time _distance _location _impactpoint _normal) (Collision|BreakHitResult _hitresult))
  (bind _actualvelocity (Variables|Default|GetActualVelocity))
  (bind _returnvalue_2 (Math|Vector|VectorLength _actualvelocity))
  (bind _returnvalue_3 (Ability|MakeOutgoingGameplayEffectSpec "/RocketDrillGameplay/Gameplay/Items/GameplayEffects/GE_RocketDrill_Damage_DirectHit_AlwaysDestroy.GE_RocketDrill_Damage_DirectHit_AlwaysDestroy_C"))
  (Variables|Default|SetActualVelocity (/ (- (Transformation|GetWorldLocation (Variables|/Script/Engine.Character|GetMesh (Variables|Default|GetPlayerPawn))) (Variables|Default|GetLastPosition)) (Utilities|Time|GetWorldDeltaSeconds)))
  (Utilities|FlowControl|Sequence
    (:then_0
      (if (Utilities|IsTraversalVersion)
        (Utilities|FlowControl|Sequence
          (:then_0
            (if _bwalkablefloor
              (Ability|HasAuthority
                (:False
                  (if (> _returnvalue_2 4e+02)
                    (bind (_outhit _returnvalue_4) (Collision|SphereTraceByProfile (Transformation|GetActorLocation (Variables|Default|GetPlayerPawn)) (+ (Transformation|GetActorLocation (Variables|Default|GetPlayerPawn)) (Utilities|Operators|Subtract (* (Math|Vector|Normalize (Math|Vector|ProjectVectorontoPlane (Math|Vector|Normalize2D(Vector) _actualvelocity) _normal)) _returnvalue_2) (Math|Vector|MakeVector 0.0 0.0 (Math|Float|SelectFloat 1.5e+02 4.5e+02 (GameplayTags|HasMatchingGameplayTag (Variables|Default|GetPlayerPawn) "(TagName=\"Granted.Athena.RocketDrill.Boosting\")"))))) 4e+01 "FortPawnCapsule" false (Utilities|Array|MakeArray (Variables|Default|GetPlayerPawn)) "None" true "(R=1.000000,G=0.000000,B=0.000000,A=1.000000)" "(R=0.000000,G=1.000000,B=0.000000,A=1.000000)" 0.05))
                    (if (not _returnvalue_4)
                      (Ability|SendGameplayEvent "(TagName=\"Event.Athena.RocketDrill.IncomingPeak\")"))))))))))
    (:then_1
      (bind _outhits (Collision|MultiSphereTraceByProfile (Transformation|GetActorLocation (Variables|Default|GetPlayerPawn)) (+ (Transformation|GetActorLocation (Variables|Default|GetPlayerPawn)) (* (Math|Vector|Normalize (Variables|Default|GetActualVelocity)) (ScalableFloat|GetValueatLevel (Variables|Default|GetSF_EnvironmentDamageTrace_Distance)))) (ScalableFloat|GetValueatLevel (Variables|Default|GetSF_EnvironmentDamageTrace_Radius)) "OverlapAll" false (Utilities|Array|MakeArray (Variables|Default|GetPlayerPawn)) "None" true "(R=1.000000,G=0.000000,B=0.000000,A=1.000000)" "(R=0.000000,G=1.000000,B=0.000000,A=1.000000)" 0.2))
      (Variables|Default|SetHitAnyBuildingActors false)
      (for _array_element _outhits
        (bind _asfort_pawn (Utilities|Casting|CastToFortPawn (.HitActor (Collision|BreakHitResult _array_element)))
          (:then
            (if (not (GameplayTags|HasMatchingGameplayTag _asfort_pawn "(TagName=\"Granted.Athena.RocketDrill.HitByDrill\")"))
              (bind (_returnvalue_x _returnvalue_y) (Math|Vector|BreakVector (* (Math|Vector|Normalize (- (Transformation|GetActorLocation _asfort_pawn) (Transformation|GetActorLocation (Variables|Default|GetPlayerPawn)))) 2e+03)))
              (Character|LaunchCharacter _asfort_pawn (Math|Vector|MakeVector _returnvalue_x _returnvalue_y 1.2e+03) false false)
              (Ability|GameplayEffect|AssignTagSetbyCallerMagnitude _returnvalue_3 "(TagName=\"SetByCaller.AbilityDamage\")" 3e+01)
              (Ability|ApplyGameplayEffectSpecToTarget _returnvalue_3 (Ability|TargetData|AbilityTargetDatafromActor _asfort_pawn))
              (Ability|ApplyGameplayEffectToTarget (Ability|TargetData|AbilityTargetDatafromActor _asfort_pawn) "/RocketDrillGameplay/Gameplay/Items/GameplayEffects/GE_RocketDrill_WasHit.GE_RocketDrill_WasHit_C")))
          (:CastFailed
            (bind _asbuilding_actor (Utilities|Casting|CastToBuildingActor (.HitActor (Collision|BreakHitResult _array_element)))
              (:then
                (if (Variables|/Script/Engine.Actor|GetbCanBeDamaged _asbuilding_actor)
                  (GameplayEffects|ApplyGameplayEffectToTarget :self (Ability|GetFortAbilitySystemComponent (Variables|Default|GetPlayerPawn)) :GameplayEffectClass (select (or (GameplayTags|HasMatchingGameplayTag (Variables|Default|GetPlayerPawn) "(TagName=\"Granted.Athena.RocketDrill.Boosting\")") (.bSuccess (Utilities|Casting|CastToBuildingProp _asbuilding_actor))) "/RocketDrillGameplay/Gameplay/Items/GameplayEffects/GE_RocketDrill_Damage_DirectHit_AlwaysDestroy.GE_RocketDrill_Damage_DirectHit_AlwaysDestroy_C" "/RocketDrillGameplay/Gameplay/Items/GameplayEffects/GE_RocketDrill_DamageOverTime.GE_RocketDrill_DamageOverTime_C") :Target (Ability|GetFortAbilitySystemComponent _asbuilding_actor) :Level 0.0)
                  (if (GameplayTags|HasAnyMatchingGameplayTags _asbuilding_actor (Variables|Default|GetStructureTag))
                    (Variables|Default|SetHitAnyBuildingActors true))))
              (:CastFailed
                (Utilities|Casting|CastToFortVehicleInterface (.HitActor (Collision|BreakHitResult _array_element)))
                (GameplayEffects|ApplyGameplayEffectToTarget :self (Ability|GetFortAbilitySystemComponent (Variables|Default|GetPlayerPawn)) :GameplayEffectClass "/RocketDrillGameplay/Gameplay/Items/GameplayEffects/GE_RocketDrill_VehicleDamage.GE_RocketDrill_VehicleDamage_C" :Target (Ability|GetFortAbilitySystemComponent (.HitActor (Collision|BreakHitResult _array_element))) :Level 0.0)
                (Audio|BroadcastAbilityImpactatLocation 0 (.HitActor (Collision|BreakHitResult _array_element)) (* (CallFunction|GetKnockbackDirection 0.7 0.5) 1e+03) 7.8e+02) ; This is the secret node that launches vehicles. Note it needs the 780 radius to work.
              )))))
      (if (and (Variables|Default|GetHitAnyBuildingActors) (not (GameplayTags|HasMatchingGameplayTag (Variables|Default|GetPlayerPawn) "(TagName=\"Granted.Athena.RocketDrill.Boosting\")")))
        (Ability|ApplyGameplayEffectToOwner "/RocketDrillGameplay/Gameplay/Items/GameplayEffects/GE_RocketDrill_BlockedByObstacle.GE_RocketDrill_BlockedByObstacle_C")
        (FailedAction|UnabletoPerformAction (Variables|Default|GetPlayerPawn) false (GameplayTags|MakeGameplayTagContainerfromTag "(TagName=\"Activate.Failed.RocketDrill.HitWall\")")) ; show reason for failing on screen if one is given
      ))
    (:then_2
      (Variables|Default|SetLastPosition (Transformation|GetWorldLocation (Variables|/Script/Engine.Character|GetMesh (Variables|Default|GetPlayerPawn)))))))

(event Ability|EventOnEndAbility
  (params (bWasCancelled :type Boolean))
  (bind (_returnvalue_x _returnvalue_y) (Math|Vector|BreakVector (* (Variables|/Script/Engine.MovementComponent|GetVelocity (Variables|/Script/Engine.Character|GetCharacterMovement (Variables|Default|GetPlayerPawn))) 0.5)))
  (|CallParentFunction bWasCancelled)
  (Pawn|Components|CharacterMovement|SetWalkableFloorAngle (Variables|/Script/Engine.Character|GetCharacterMovement (Variables|Default|GetPlayerPawn)) 4e+01)
  (Class|CameraModifierRocketDrillTraversalVersion|WindDownandRemove (Variables|Default|GetCameraMod))
  (bind _asfort_movement_comp_character (Utilities|Casting|CastToFortMovementComp_Character (Variables|/Script/Engine.Character|GetCharacterMovement (Variables|Default|GetPlayerPawn))))
  (NetTolerance|SetUseHighAccelerationTolerance _asfort_movement_comp_character false "RocketDrillMovement") ; TODO this is to temporarily hide net corrections, will need to be removed eventually
  (Utilities|Time|ClearandInvalidateTimerbyHandle (Variables|Default|GetSlopeCheckTimer))
  (Utilities|Time|ClearandInvalidateTimerbyHandle (Variables|Default|GetDurabilityTimerHandle))
  (Collision|SetCollisionResponsetoChannel (Variables|/Script/Engine.Character|GetCapsuleComponent (Variables|Default|GetPlayerPawn)) "ECC_GameTraceChannel18" "ECR_Overlap")
  (Ability|RemoveGameplayCueFromOwner (Variables|Default|GetTimeoutCue))
  (Ability|RemoveGameplayCueFromOwner (Variables|Default|GetBoostingCue))
  (Character|LaunchCharacter (Variables|Default|GetPlayerPawn) (Math|Vector|MakeVector _returnvalue_x _returnvalue_y 1e+03) true true)
  (Ability|ApplyGameplayEffectToOwner "/RocketDrillGameplay/Gameplay/Items/GameplayEffects/GE_SkidToStop.GE_SkidToStop_C")
  (Ability|ApplyGameplayEffectToOwner "/Game/Athena/Items/Parents/FallImmunity/GE_Athena_TillLandFallDamageImmunity_NoFX_Grant_Parent.GE_Athena_TillLandFallDamageImmunity_NoFX_Grant_Parent_C")
  (Ability|CommitAbilityCooldown false true))

(fn CheckCollisionAngle
  (params (HitResult :type Hit%20Result%20Structure))
  (bind (_bblockinghit _binitialoverlap _time _distance _location _impactpoint _normal) (Collision|BreakHitResult HitResult))
  ; TODO: Improve logic. Want to use a blend of angle of incidence and distance from the vault to have some prediction
  (return (> (Math|Vector|DotProduct (* _normal -1.0) (Transformation|GetActorForwardVector (Variables|Default|GetPlayerPawn))) 0.96)))

(fn GetKnockbackDirection
  (params (UpAmount :type Float%20%28double-precision%29) (RightAmount :type Float%20%28single-precision%29))
  (return (Utilities|Operators|Add (Math|Vector|Normalize2D(Vector) (Math|Vector|GetForwardVector (Pawn|GetBaseAimRotation (Variables|Default|GetPlayerPawn)))) (Math|Vector|MakeVector 0.0 0.0 UpAmount) (* (Math|Vector|CrossProduct (Math|Vector|Normalize2D(Vector) (Math|Vector|GetForwardVector (Pawn|GetBaseAimRotation (Variables|Default|GetPlayerPawn)))) "0.000000,0.000000,-1.000000") RightAmount))))
