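; Gameplay ability for the Rocket Drill item's primary fire, which requests the drill's activation.
; On activate, after calling the parent, it checks the activating pawn for the Gameplay.Action.Player.Medium tag.
; Without that tag, it sends the RocketDrill StartActivation gameplay event and ends immediately; with it, this graph does nothing further.
; End ability only calls the parent.
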(event Ability|EventActivateAbility
  (|CallParentFunction)
  (if (not (GameplayTags|HasMatchingGameplayTag (Ability|GetActivatingPawn) "(TagName=\"Gameplay.Action.Player.Medium\")"))
    (Ability|SendGameplayEvent "(TagName=\"Event.Athena.RocketDrill.StartActivation\")")
    (Ability|EndAbility)))

(event Ability|EventOnEndAbility
  (params (bWasCancelled :type Boolean))
  (|CallParentFunction bWasCancelled))
