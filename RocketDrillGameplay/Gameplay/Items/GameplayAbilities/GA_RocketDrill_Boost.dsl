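; Gameplay ability that triggers the Rocket Drill item's boost on the owning player.
; On activation it sends the RocketDrill Boost gameplay event and applies the boosting gameplay effect to the owner.
; It then plays the Rocket Ram fire sound, borrowed from that item, at the player pawn's location and ends immediately.
; Any boost movement or duration logic sits outside this graph.
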(event Ability|EventActivateAbility
  (Ability|SendGameplayEvent "(TagName=\"Event.Athena.RocketDrill.Boost\")")
  (Ability|ApplyGameplayEffectToOwner "/RocketDrillGameplay/Gameplay/Items/GameplayEffects/GE_Boosting_RocketDrill.GE_Boosting_RocketDrill_C")
  (Audio|PlaySoundatLocation "/RocketRamGameplay/Items/RocketRam/Sounds/RocketRam_Fire_Cue.RocketRam_Fire_Cue" (Transformation|GetActorLocation (CallFunction|GetAvatarasFortPlayerPawn)))
  (Ability|EndAbility))
