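; Gameplay ability for the Rocket Drill item's fire input, which signals the drill to stop when the input is released.
; On activation it starts input press and release tasks; the press branch leads to no actions.
; On release it waits for a client-server net sync, then sends the RocketDrill Stop gameplay event.
; The event is skipped when the activating pawn has the UsingJetpackBurstVersion tag.
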(event Ability|EventActivateAbility
  (Utilities|FlowControl|Sequence
    (:then_0
      (Ability|Tasks|WaitInputPress false
        (:OnPress
          (Utilities|FlowControl|Sequence
            (:then_1)))))
    (:then_1
      (Ability|Tasks|WaitInputRelease false
        (:OnRelease
          (Utilities|FlowControl|Sequence
            (:then_0
              _)
            (:then_1
              _))
          (Ability|Tasks|WaitNetSync "BothWait"
            (:OnSync
              (if (not (GameplayTags|HasMatchingGameplayTag (Ability|GetActivatingPawn) "(TagName=\"Granted.Athena.RocketDrill.UsingJetpackBurstVersion\")"))
                (Ability|SendGameplayEvent "(TagName=\"Event.Athena.RocketDrill.Stop\")")))))))))
