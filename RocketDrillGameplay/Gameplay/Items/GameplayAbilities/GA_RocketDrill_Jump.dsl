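; GA_RocketDrill_Jump is a gameplay ability that tracks jump input for the Rocket Drill item.
; On activation it sets up input press and release wait tasks together.
; Press applies GE_Jumping_RocketDrill to the owner, stores its handle and sends the JumpInput.Start gameplay event.
; Release removes that effect by handle and sends JumpInput.End; the jump movement itself is not implemented here.
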(event Ability|EventActivateAbility
  (Utilities|FlowControl|Sequence
    (:then_0
      (Ability|Tasks|WaitInputPress false
        (:OnPress
          (Utilities|FlowControl|Sequence
            (:then_0
              _)
            (:then_1
              _))
          (bind _returnvalue (Ability|ApplyGameplayEffectToOwner "/RocketDrillGameplay/Gameplay/Items/GameplayEffects/GE_Jumping_RocketDrill.GE_Jumping_RocketDrill_C"))
          (Variables|Default|SetGE_Jumping _returnvalue)
          (Ability|SendGameplayEvent "(TagName=\"Event.Athena.RocketDrill.JumpInput.Start\")"))))
    (:then_1
      (Ability|Tasks|WaitInputRelease false
        (:OnRelease
          (Utilities|FlowControl|Sequence
            (:then_0
              _)
            (:then_1
              _))
          (Ability|RemoveGameplayEffectFromOwnerWithHandle (Variables|Default|GetGE_Jumping))
          (Ability|SendGameplayEvent "(TagName=\"Event.Athena.RocketDrill.JumpInput.End\")"))))))
